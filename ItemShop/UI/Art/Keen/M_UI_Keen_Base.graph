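# UI-domain base material for Keen art in the ItemShop, built around the Keen_Red texture parameter.
# Aspect-fit style math scales and centers widget UVs before sampling, but each pixel-size component is compared or divided by itself, so UVs effectively pass through.
# The sampled RGB feeds a 0.7-alpha lerp.
# A screen-aligned radial gradient, CorruptSeries color parameters and a second lerp sit unconnected, and no material output links appear.
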Material /ItemShop/UI/Art/Keen/M_UI_Keen_Base
graph {
 properties {bUsedWithStaticMesh=false, materialDomain=MD_UI, pixelDepthOffsetMode=PDOM_Legacy, usedShadingModels=Default Lit}
 nodes [
  n1 = Constant() {r=1}
  n2 = MaterialFunctionCall() {materialFunction=&GetUserInterfaceUV}
  n3 = MaterialFunctionCall(Float2=@n2.Pixel Size) {materialFunction=&BreakOutFloat2Components}
  n4 = Divide(A=@n3.G, B=@n3.G)
  n5 = OneMinus(None=@n4)
  n6 = Add(A=Multiply(A=@n2.Normalized UV, B=AppendVector(A=@n1, B=@n4)), B=AppendVector(A=Constant(), B=Multiply(A=@n5) {constB=0.5}))
  n7 = MaterialFunctionCall(Float2=MaterialFunctionCall() {materialFunction=&GetUserInterfaceUV}.Pixel Size) {materialFunction=&BreakOutFloat2Components}
  n8 = Constant()
  n9 = If(A=@n7.R, B=@n7.R, "A > B"=@n8, "A == B"=@n8, "A < B"=Constant() {r=1})
  n10 = MaterialFunctionCall() {materialFunction=&GetUserInterfaceUV}
  n11 = MaterialFunctionCall(Float2=@n10.Pixel Size) {materialFunction=&BreakOutFloat2Components}
  n12 = Divide(A=@n11.R, B=@n11.R)
  n13 = Constant() {r=1}
  n14 = OneMinus(None=@n12)
  n15 = Add(A=Multiply(A=@n10.Normalized UV, B=AppendVector(A=@n12, B=@n13)), B=AppendVector(A=Multiply(A=@n14) {constB=0.5}, B=Constant()))
  LinearInterpolate(B=TextureSampleParameter2D(UVs=LinearInterpolate(A=@n6, B=@n15, Alpha=@n9)) {parameterName=Keen_Red, texture=/OfferCatalog/Art/Keen/T_UI_Keen}.RGB) {constAlpha=0.7}
  LinearInterpolate()
  MaterialFunctionCall(
   UVs=MaterialFunctionCall() {materialFunction=&ScreenAlignedUVs}."X 100%, Y 100%"
   CenterPosition=Constant2Vector() {g=0.5, r=0.5}
   Radius=ScalarParameter() {defaultValue=0.5, parameterName=RadialRadius_02}
   Density=ScalarParameter() {defaultValue=1.0452, parameterName=RadialDensity_02}
  ) {materialFunction=&RadialGradientExponential}
  VectorParameter() {group=CorruptSeries, parameterName=CellColor, defaultValue=<rgba 0 0 0 1>}
  VectorParameter() {group=CorruptSeries, parameterName=BG_Color, defaultValue=<rgba 1 1 1 1>}
 ]
}
